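; PCG graph that merges local Biome Core data, resolves priorities and spawns the biome content.
; Generated points are moved to Grid128, culled, priority-sorted, differenced and assembly-filtered, with SkipDifference points bypassing differencing.
; Tagged custom biome data joins afterwards, then everything feeds the spawning subgraph and an overridable PostProcess graph.
; Optional branches clear collision, debug-display biomes, and output partitioned mesh scatter and weighted ground scatter data for runtime generation.
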(graph BiomeCore
  (params (PostProcess :type (Object /Script/PCG.PCGGraphInterface)) (DisallowCollisions :type Bool) (Debug-DisplayBiomes :type Bool) (OutputDataForRuntimeGeneration :type Bool :default true) (ForceCreateTargetActors :type Bool) (TargetActorsClass :type (Struct /Script/CoreUObject.SoftClassPath) :default (struct :AssetPath (struct :PackageName /Script/Engine :AssetName Actor))) (MergeDifferentDataInSameInstancedComponents :type Bool) (SkeletalMeshTransformProviderData :type (SoftObject /Script/Engine.TransformProviderData)))
  (outputs (MeshScatter:AssetPaths :type Param) (MeshScatter:WorldTransforms :type Point) (GroundScatter:InfluencePoints :type Point) (GroundScatter:Biomes :type Point) (GroundScatter:WeightedAssets :type Param))
  (bind Gather_3 (Gather))
  (bind Branch_8 (Branch :bOutputToB-param DisallowCollisions))
  (bind AttributeMathsOp_0 (Attribute%20Maths%20Op :InputSource1 AllowCollision :Operation Set))
  (bind Gather_1 (Gather))
  ;[0 0.5 1 1 32] Custom Biome Data
  (bind GetActorData_4 (Get%20Actor%20Data :ActorSelector (struct :ActorFilter AllWorldActors :bMustOverlapSelf true :ActorSelectionTag PCG_CustomBiomeData :bSelectMultiple true) :bComponentsMustOverlapSelf false :Mode GetDataFromPCGComponent))
  (bind AddTags_0 (Add%20Tags :bTokenizeOnWhiteSpace true :TagsToAdd "Custom"))
  (bind ToPoint_2 (To%20Point))
  (bind HiGenGridSize_4 (Change%20Grid%20Size :HiGenGridSize Grid128))
  (bind Subgraph_6 (Subgraph :Graph /PCGBiomeCore/Core/BiomeCore_CullPointsOutsidePartitionActors.BiomeCore_CullPointsOutsidePartitionActors))
  ;[0 0 0 1 32] Spawning
  (bind Subgraph_0 (Subgraph :Graph /PCGBiomeCore/Core/BiomeCore_Spawning.BiomeCore_Spawning :ForceCreateTargetActors ForceCreateTargetActors :MergeDifferentDataInSameInstancedComponents MergeDifferentDataInSameInstancedComponents :SkeletalMeshTransformProviderData SkeletalMeshTransformProviderData :TargetActorsClass TargetActorsClass))
  ;[0 0 0] Post Process Graph
  (bind Subgraph_51 (Subgraph :SubgraphOverride-param PostProcess))
  ;[0 0 0 1 32] Get All Local Biome Core Data
  (bind GetActorData_5 (Get%20Actor%20Data :ActorSelector (struct :ActorFilter AllWorldActors :bMustOverlapSelf true :ActorSelectionTag PCG_BiomeActor :bSelectMultiple true) :bMergeSinglePointData true :ComponentSelector (struct :ComponentSelectionTag PCG_LocalBiomeCore) :ExpectedPins [BiomePoints BiomeLocalCache] :Mode GetActorReference))
  (bind SortAttributes_0 (Sort%20Attributes :bUseStableSort false :SortAttributes [(struct :InputSource ActorReference)]))
  (bind GetActorData_6 (Get%20Actor%20Data :ActorSelector (struct :ActorFilter FromInput :ActorReferenceSelector ActorReference) :ComponentSelector (struct :ComponentSelectionTag PCG_LocalBiomeCore) :ExpectedPins [GeneratedPoints RuntimeMeshScatter:AssetTables RuntimeGroundScatter:Biome RuntimeGroundScatter:AssetsAndDefinitions] :Mode GetDataFromPCGComponent))
  (bind FilterDataByType_2 (Filter%20Data%20By%20Type :TargetType (struct :Ids [(struct :Struct /Script/PCG.PCGDataTypeInfoPoint)])))
  (bind HiGenGridSize_0 (Change%20Grid%20Size :HiGenGridSize Grid128))
  (bind CullPointsOutsideActorBounds_0 (Cull%20Points%20Outside%20Actor%20Bounds))
  (bind SortTags_2 (Sort%20Data%20By%20Tag%20Value :Tag RecursionLevel))
  (bind SortTags_0 (Sort%20Data%20By%20Tag%20Value :Tag PRIO))
  (bind SortTags_1 (Sort%20Data%20By%20Tag%20Value :Tag BIOMEPRIO))
  ;;[0.15 0.15 0.15 0.5] Runtime Inputs
  (bind FilterDataByType_3 (Filter%20Data%20By%20Type :TargetType (struct :Ids [(struct :Struct /Script/PCG.PCGDataTypeInfoPoint)])))
  (bind FilterDataByType_4 (Filter%20Data%20By%20Type :TargetType (struct :Ids [(struct :Struct /Script/PCG.PCGDataTypeInfoParam)])))
  (bind FilterDataByType_5 (Filter%20Data%20By%20Type :TargetType (struct :Ids [(struct :Struct /Script/PCG.PCGDataTypeInfoParam)])))
  ;[0 0 0 1 32] Difference
  (bind FilterDataByTag_39 (Filter%20Data%20By%20Tag :bTokenizeOnWhiteSpace true :Operation RemoveTagged :SelectedTags "SkipDifference"))
  (bind Loop_16 (Loop :bTokenizeOnWhiteSpace true :Graph /PCGBiomeCore/Core/BiomeCore_DifferenceByPriority.BiomeCore_DifferenceByPriority))
  (bind CreateConstant_0 (Create%20Constant :AttributeTypes (struct :Type Boolean :BoolValue true)))
  (bind Subgraph_4 (Subgraph :Graph /PCGBiomeCore/Core/BiomeCore_PostDifference_FilterAssemblies.BiomeCore_PostDifference_FilterAssemblies))
  ;[0 0 0 1 32] Runtime Outputs
  (bind Branch_1 (Branch :bOutputToB-param OutputDataForRuntimeGeneration))
  (bind DataNum_0 (Data%20Count))
  (bind DataNum_1 (Data%20Count))
  ;;[0.15 0.15 0.15 0.5] Mesh Scattering
  (bind Branch_5 (Branch))
  (bind AttributeFilter_3 (Filter%20Attribute%20Elements :AttributeTypes (struct :Type Integer32) :bUseConstantThreshold true :bWarnOnDataMissingAttribute false :Operator NotEqual :TargetAttribute RuntimeAssemblyHashValue :ThresholdAttribute @Last))
  (bind Branch_3 (Branch :bOutputToB-param OutputDataForRuntimeGeneration))
  (bind RemoveEmptyData_1 (Remove%20Empty%20Data))
  (bind DeleteAttributes_2 (Delete%20Attributes :Operation KeepSelectedAttributes :SelectedAttributes "RuntimeAssemblyHashValue"))
  (bind AttributePartition_3 (Attribute%20Partition :PartitionAttributeSelectors [RuntimeAssemblyHashValue]))
  (bind Subgraph_3 (Subgraph :Graph /PCGBiomeCore/Core/BiomeCore_ValidateAssetPaths.BiomeCore_ValidateAssetPaths))
  ;;[0.15 0.15 0.15 0.5] Ground Scattering
  (bind Branch_6 (Branch))
  (bind AttributeFilter_4 (Filter%20Attribute%20Elements :AttributeTypes (struct :Type Integer32) :bUseConstantThreshold true :bWarnOnDataMissingAttribute false :Operator NotEqual :TargetAttribute InfluenceType :ThresholdAttribute @Last))
  (bind Branch_7 (Branch))
  (bind Branch_2 (Branch :bOutputToB-param OutputDataForRuntimeGeneration))
  (bind RemoveEmptyData_0 (Remove%20Empty%20Data))
  (bind DeleteAttributes_3 (Delete%20Attributes :Operation KeepSelectedAttributes :SelectedAttributes "InfluenceType,InfluenceRadius"))
  (bind Branch_0 (Branch :bOutputToB-param OutputDataForRuntimeGeneration))
  (bind MergeAttributes_2 (Merge%20Attributes))
  (bind AttributePartition_0 (Attribute%20Partition :PartitionAttributeSelectors [GeneratorGraph]))
  (bind Loop_3 (Loop :Graph /PCGBiomeCore/Runtime/GroundScatter/GroundScatterWeightsDistributionLoop.GroundScatterWeightsDistributionLoop))
  ;[1 0.3333 0 0.2] Display Biome Cache
  (bind Branch_4 (Branch :bOutputToB-param Debug-DisplayBiomes))
  (bind Debug_1 (Debug))
  (bind ExtentsModifier_1 (Extents%20Modifier :Mode Add))
  (bind Debug_2 (Debug))
  (reroute FinalPoints Gather_1)
  (reroute GroundScatterAssetsAndDefinitions FilterDataByType_5.InsideFilter)
  (reroute GroundScatterBiomes FilterDataByType_3.InsideFilter)
  (reroute MeshScatterAssetTables FilterDataByType_4.InsideFilter)
  (connect GetActorData_4 Out AddTags_0 In)
  (connect Branch_5 Output%20B AttributeFilter_3 In)
  (connect Branch_6 Output%20B AttributeFilter_4 In)
  (connect Branch_8 Output%20B AttributeMathsOp_0 InA)
  (connect MergeAttributes_2 Out AttributePartition_0 In)
  (connect DeleteAttributes_2 Out AttributePartition_3 In)
  (connect GroundScatterAssetsAndDefinitions Out Branch_0 In)
  (connect FinalPoints Out Branch_1 In)
  (connect Branch_7 Output%20B Branch_2 In)
  (connect MeshScatterAssetTables Out Branch_3 In)
  (connect GroundScatterBiomes Out Branch_4 In)
  (connect DataNum_0 Out Branch_5 bOutputToB)
  (connect Branch_1 Output%20B Branch_5 In)
  (connect DataNum_1 Out Branch_6 bOutputToB)
  (connect Branch_1 Output%20B Branch_6 In)
  (connect DataNum_1 Out Branch_7 bOutputToB)
  (connect GroundScatterBiomes Out Branch_7 In)
  (connect Gather_3 Out Branch_8 In)
  (connect HiGenGridSize_0 Out CullPointsOutsideActorBounds_0 In)
  (connect MeshScatterAssetTables Out DataNum_0 In)
  (connect GroundScatterAssetsAndDefinitions Out DataNum_1 In)
  (connect Branch_4 Output%20B Debug_1 In)
  (connect ExtentsModifier_1 Out Debug_2 In)
  (connect RemoveEmptyData_1 Out DeleteAttributes_2 In)
  (connect RemoveEmptyData_0 Out DeleteAttributes_3 In)
  (connect Branch_4 Output%20B ExtentsModifier_1 In)
  (connect SortTags_1 Out FilterDataByTag_39 In)
  (connect GetActorData_6 GeneratedPoints FilterDataByType_2 In)
  (connect GetActorData_6 RuntimeGroundScatter:Biome FilterDataByType_3 In)
  (connect GetActorData_6 RuntimeMeshScatter:AssetTables FilterDataByType_4 In)
  (connect GetActorData_6 RuntimeGroundScatter:AssetsAndDefinitions FilterDataByType_5 In)
  (connect Branch_8 Output%20A Gather_1 In)
  (connect AttributeMathsOp_0 Out Gather_1 In2)
  (connect Subgraph_4 Out Gather_3 In)
  (connect Subgraph_6 PointsOut Gather_3 In2)
  (connect SortAttributes_0 Out GetActorData_6 In)
  (connect FilterDataByType_2 InsideFilter HiGenGridSize_0 In)
  (connect ToPoint_2 Out HiGenGridSize_4 In)
  (connect AttributePartition_0 Out Loop_3 In)
  (connect FilterDataByTag_39 InsideFilter Loop_16 SourcePoints)
  (connect Branch_0 Output%20B MergeAttributes_2 In)
  (connect AttributeFilter_4 InsideFilter RemoveEmptyData_0 In)
  (connect AttributeFilter_3 InsideFilter RemoveEmptyData_1 In)
  (connect GetActorData_5 Out SortAttributes_0 In)
  (connect SortTags_2 Out SortTags_0 In)
  (connect SortTags_0 Out SortTags_1 In)
  (connect CullPointsOutsideActorBounds_0 Out SortTags_2 In)
  (connect FinalPoints Out Subgraph_0 In)
  (connect Branch_3 Output%20B Subgraph_3 In)
  (connect Loop_16 RemainingPoints Subgraph_4 DifferencePoints)
  (connect CreateConstant_0 Out Subgraph_4 Enabled)
  (connect FilterDataByTag_39 OutsideFilter Subgraph_4 SkipFiltersPoints)
  (connect HiGenGridSize_4 Out Subgraph_6 PointsIn)
  (connect Subgraph_0 Out Subgraph_51 In)
  (connect AddTags_0 Out ToPoint_2 In)
  (output GroundScatter:Biomes Branch_2.Output%20B)
  (output GroundScatter:InfluencePoints DeleteAttributes_3)
  (output GroundScatter:WeightedAssets Loop_3)
  (output MeshScatter:AssetPaths Subgraph_3)
  (output MeshScatter:WorldTransforms AttributePartition_3))
