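; Base actor for PCG biomes that controls a local Biome Core PCG component and exposes its enabled state as an actor tag.
; The construction script adds or removes the "Enabled" tag to match the biome's Enabled variable.
; Generate forces generation on the local component.
; ToggleBypassFilters and ToggleLocalPreview flip the debug filter-bypass and local preview bool parameters on that component's graph instance.
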(fn Generate
  (PCG|Generate (Variables|Default|GetPCG_LocalBiomeCore) true))

(fn ToggleBypassFilters
  (bind _pcg_localbiomecore (Variables|Default|GetPCG_LocalBiomeCore))
  (bind _graphinstance (Variables|/Script/PCG.PCGComponent|GetGraphInstance _pcg_localbiomecore))
  (bind _returnvalue (PCG|GraphParameters|GetBoolParameter _graphinstance "Debug-BypassFilters"))
  (if _returnvalue
    (PCG|GraphParameters|SetBoolParameter _graphinstance "Debug-BypassFilters")
    (else
      (PCG|GraphParameters|SetBoolParameter _graphinstance "Debug-BypassFilters" true))))

(fn ToggleLocalPreview
  (bind _pcg_localbiomecore (Variables|Default|GetPCG_LocalBiomeCore))
  (bind _graphinstance (Variables|/Script/PCG.PCGComponent|GetGraphInstance _pcg_localbiomecore))
  (bind _returnvalue (PCG|GraphParameters|GetBoolParameter _graphinstance "LocalPreview"))
  (if _returnvalue
    (PCG|GraphParameters|SetBoolParameter _graphinstance "LocalPreview")
    (else
      (PCG|GraphParameters|SetBoolParameter _graphinstance "LocalPreview" true))))

(fn ConstructionScript
  (bind _returnvalue (Utilities|Name|MakeLiteralName "Enabled"))
  (bind _tags (Variables|Actor|GetTags))
  (if (Variables|Biome|GetEnabled)
    (Utilities|Array|AddUnique _tags _returnvalue)
    (else
      (Utilities|Array|RemoveItem _tags _returnvalue)))
  (Variables|Actor|SetTags _tags))
