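; Blueprint wrapper for the PCG Biome Core, exposing generation and a cache debug toggle for its BiomeCore PCG component.
; Generate calls PCG Generate on that component with the force flag set.
; ToggleCache reads the Debug-DisplayBiomeCache bool parameter on the component's graph instance and writes back the opposite value.
; BeginPlay, Tick, overlap and the construction script are empty.
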(event EventBeginPlay)

(event Collision|EventActorBeginOverlap
  (params (OtherActor :type Actor%20Object%20Reference)))

(event EventTick
  (params (DeltaSeconds :type Float%20%28single-precision%29)))

(fn Generate
  (PCG|Generate (Variables|Default|GetBiomeCore) true))

(fn ToggleCache
  (bind _biomecore (Variables|Default|GetBiomeCore))
  (bind _graphinstance (Variables|/Script/PCG.PCGComponent|GetGraphInstance _biomecore))
  (bind _returnvalue (PCG|GraphParameters|GetBoolParameter _graphinstance "Debug-DisplayBiomeCache"))
  (if _returnvalue
    (PCG|GraphParameters|SetBoolParameter _graphinstance "Debug-DisplayBiomeCache")
    (else
      (PCG|GraphParameters|SetBoolParameter _graphinstance "Debug-DisplayBiomeCache" true))))

(fn ConstructionScript)
