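; Biome volume Blueprint for the PCG Biome Core plugin, holding the biome definition for a volume.
; TransferToDefaultDefinition reads the biome definition from the referenced BiomeDefinitionTemplate-based Definition, writes it into the local DefaultDefinition, then clears the Definition reference.
; This appears to migrate asset-referenced definitions to locally stored ones.
; BeginPlay, Tick, overlap and construction script are empty.
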(event EventBeginPlay)

(event Collision|EventActorBeginOverlap
  (params (OtherActor :type Actor%20Object%20Reference)))

(event EventTick
  (params (DeltaSeconds :type Float%20%28single-precision%29)))

(fn TransferToDefaultDefinition
  (Variables|Biome|Biome%28Local%29|SetDefaultDefinition (Variables|/PCGBiomeCore/BiomeDefinitions/Setup/BiomeDefinitionTemplate.BiomeDefinitionTemplate_C|GetBiomeDefinition (Variables|Biome|GetDefinition)))
  (Variables|Biome|SetDefinition 0))

(fn ConstructionScript)
