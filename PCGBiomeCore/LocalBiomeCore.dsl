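; LocalBiomeCore is a PCG graph that generates biome points for an overlapping BiomeCore, or spawns meshes, actors and skinned meshes itself in Local Preview.
; Generators from actor assets sample the local biome cache, then points get partition culling, exclusion differencing, root assets and filters.
; Points loop by AssetID through child transforms, expand assemblies, and are restored by generator.
; It also outputs runtime mesh and ground scatter data, and flags a missing BiomeCore.
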(graph LocalBiomeCore
  (params (RootPointsFilters :type (Object /Script/PCG.PCGGraphInterface) :default /PCGBiomeCore/Core/BiomeCore_Filters_Inst.BiomeCore_Filters_Inst) (ChildPointsFilters :type (Object /Script/PCG.PCGGraphInterface) :default /PCGBiomeCore/Core/BiomeCore_Filters_Inst.BiomeCore_Filters_Inst) (DisallowCollisions :type Bool) (Debug-BypassFilters :type Bool) (MaxChildAssetDepth :type Int :default 2) (ChildInputPointsRateMultiplier :type Int :default 10) (LocalPreview :type Bool) (LocalBiomeCache :type (Object /Script/PCG.PCGGraphInterface)) (Debug-DisplayLocalBiome :type Bool) (AssemblyBehavior :type (Object /Script/PCG.PCGGraphInterface) :default /PCGBiomeCore/AssemblyBehavior/BiomeCore_DefaultAssemblyBehavior.BiomeCore_DefaultAssemblyBehavior) (Debug-BypassAssemblyBehavior :type Bool) (DefaultTerrainGetterSetting :type (Object /Script/PCG.PCGSettingsInterface) :default /PCGBiomeCore/Setup/WorldRayHitSetting.WorldRayHitSetting))
  (outputs (GeneratedPoints :type Point) (RuntimeMeshScatter:AssetTables :type Param) (RuntimeGroundScatter:Biome :type Point) (RuntimeGroundScatter:AssetsAndDefinitions :type Param))
  (bind Difference_3 (Difference :bDiffMetadata false :DensityFunction Binary :Mode Discrete))
  (bind ToPoint_0 (To%20Point))
  (bind RemoveEmptyData_3 (Remove%20Empty%20Data))
  (bind Print%20String_0 (Print%20String :bPrefixWithNode false :bPrefixWithOwner true :bPrintToScreen true :PrintString "Bypass Global Filters is currently enabled" :PrintToScreenColor (struct :G 173 :R 255 :A 255) :PrintToScreenDuration 0.0 :Verbosity Warning))
  (bind FilterDataByTag_3 (Filter%20Data%20By%20Tag :Operation RemoveTagged :SelectedTags "Assembly"))
  (bind Branch_5 (Branch :bOutputToB-param DisallowCollisions))
  (bind AttributeMathsOp_2 (Attribute%20Maths%20Op :InputSource1 AllowCollision :Operation Set))
  (bind Branch_2 (Branch :bOutputToB-param LocalPreview))
  ;[0 0.5 1 1 32] Exclusions
  (bind GetActorData_6 (Get%20Actor%20Data :ActorSelector (struct :ActorFilter AllWorldActors :bMustOverlapSelf true :ActorSelectionTag PCG_BiomeExclusion :bSelectMultiple true :bIgnoreSelfAndChildren true) :bComponentsMustOverlapSelf false))
  (bind FilterByTag_5 (Filter%20Data%20By%20Tag :bTokenizeOnWhiteSpace true :SelectedTags "BiomeExclusion"))
  (bind FilterByType_7 (Filter%20Data%20By%20Type :TargetType (struct :Ids [(struct :Struct /Script/PCG.PCGDataTypeInfoSpline)])))
  (bind SplineSampler_1 (Spline%20Sampler :SamplerParams (struct :Dimension OnInterior :InteriorSampleSpacing 4e+02 :InteriorBorderSampleSpacing 4e+02 :bUnbounded true :PointSteepness 1.0)))
  (bind BoundsModifier_2 (Bounds%20Modifier :BoundsMax (1.0 1.0 1e+02) :BoundsMin (1.0 1.0 1e+02)))
  (bind FilterByType_8 (Filter%20Data%20By%20Type :TargetType (struct :Ids [(struct :Struct /Script/PCG.PCGDataTypeInfoPrimitive)])))
  (bind FilterDataByTag_9 (Filter%20Data%20By%20Tag :bTokenizeOnWhiteSpace true :SelectedTags "BiomePath"))
  (bind FilterDataByType_7 (Filter%20Data%20By%20Type :TargetType (struct :Ids [(struct :Struct /Script/PCG.PCGDataTypeInfoSpline)])))
  ;[0 0 0 1 32] Generators
  (bind GetActorProperty_1 (Get%20Actor%20Property :ActorSelector (struct :ActorSelectionTag PCG_BiomeActor :bSelectMultiple true) :bAlwaysRequeryActors false :bProcessAllComponents false :bSanitizeOutputAttributeName false :bTrackActorsOnlyWithinBounds true :OutputAttributeName BiomeBlendingRange :PropertyName BlendingRange))
  (bind GetActorProperty_3 (Get%20Actor%20Property :ActorSelector (struct :ActorSelectionTag PCG_BiomeActor :bSelectMultiple true) :bAlwaysRequeryActors false :bProcessAllComponents false :bSanitizeOutputAttributeName false :bTrackActorsOnlyWithinBounds true :OutputAttributeName BiomeBlendingNoise :PropertyName BlendingNoise))
  (bind GetActorProperty_4 (Get%20Actor%20Property :ActorSelector (struct :ActorSelectionTag PCG_BiomeActor :bSelectMultiple true) :bAlwaysRequeryActors false :bProcessAllComponents false :bSanitizeOutputAttributeName false :bTrackActorsOnlyWithinBounds true :OutputAttributeName BiomeBlendingDensity :PropertyName BlendingDensity))
  (bind GetActorProperty_7 (Get%20Actor%20Property :ActorSelector (struct :ActorSelectionTag PCG_BiomeActor :bSelectMultiple true) :bAlwaysRequeryActors false :bProcessAllComponents false :bSanitizeOutputAttributeName false :bTrackActorsOnlyWithinBounds true :OutputAttributeName Assets :PropertyName Assets))
  (bind FilterDataByTag_1 (Filter%20Data%20By%20Tag :SelectedTags "Enabled"))
  (bind GetActorProperty_2 (Get%20Actor%20Property :bAlwaysRequeryActors false :bSanitizeOutputAttributeName false :bTrackActorsOnlyWithinBounds true :PropertyName LocalAssets))
  (bind FilterDataByTag_2 (Filter%20Data%20By%20Tag :SelectedTags "Enabled"))
  (bind Subgraph_133 (Subgraph :Graph /PCGBiomeCore/Core/BiomeCore_ResolveGeneratorsFromAssets.BiomeCore_ResolveGeneratorsFromAssets))
  (bind Loop_6 (Loop :bTokenizeOnWhiteSpace true :Graph /PCGBiomeCore/Core/BiomeCore_RunGenerators.BiomeCore_RunGenerators))
  (bind Subgraph_21 (Subgraph :Graph /PCGBiomeCore/Core/BiomeCore_CullPointsOutsidePartitionActors.BiomeCore_CullPointsOutsidePartitionActors))
  ;[0 0 0 1 32] Restore Data by Generators
  (bind CreateConstant_1 (Create%20Constant :AttributeTypes (struct :Type Boolean :BoolValue true)))
  (bind Subgraph_1 (Subgraph :Graph /PCGBiomeCore/Core/BiomeCore_RestoreDataByGenerators.BiomeCore_RestoreDataByGenerators))
  ;[0 0 0 1 32] Assign Assets
  (bind Subgraph_8 (Subgraph :Graph /PCGBiomeCore/Core/BiomeCore_AssignAssetsToPoints.BiomeCore_AssignAssetsToPoints))
  ;[1 0.3333 0 0.1] Bypass Global Filters
  (bind Branch_0 (Branch :bOutputToB-param Debug-BypassFilters))
  ;[0 0 0 1 32] Filters
  (bind FilterDataByTag_5 (Filter%20Data%20By%20Tag :bTokenizeOnWhiteSpace true :SelectedTags "SkipFilters"))
  (bind Subgraph_3 (Subgraph :Graph /PCGBiomeCore/Core/BiomeCore_Filters.BiomeCore_Filters :SubgraphOverride-param RootPointsFilters))
  (bind RemoveEmptyData_2 (Remove%20Empty%20Data))
  (bind FilterDataByType_2 (Filter%20Data%20By%20Type :TargetType (struct :Ids [(struct :Struct /Script/PCG.PCGDataTypeInfoPoint)])))
  ;[0 0 0 1 32] Transforms
  ;;[0 0 0] Root and Child Assets Recursion
  (bind MergePoints_0 (Merge%20Points))
  (bind MutateSeed_1 (Mutate%20Seed))
  (bind AttributePartition_3 (Attribute%20Partition :PartitionAttributeSelectors [AssetID]))
  (bind Loop_5 (Loop :bTokenizeOnWhiteSpace true :Graph /PCGBiomeCore/Core/BiomeCore_RootTransformLoop.BiomeCore_RootTransformLoop :Filters ChildPointsFilters :MaxIterations MaxChildAssetDepth :TransfomInputPointsRateMultiplier ChildInputPointsRateMultiplier))
  ;[0 0 0 1 32] Assemblies
  (bind Loop_1 (Loop :bTokenizeOnWhiteSpace true :Graph /PCGBiomeCore/Core/BiomeCore_AssemblyInstancer.BiomeCore_AssemblyInstancer :AssemblyBehavior AssemblyBehavior :Debug-BypassAssemblyBehavior Debug-BypassAssemblyBehavior :DefaultTerrainGetterSetting DefaultTerrainGetterSetting))
  ;[0 0 0] Local Preview Spawners
  (bind AttributeFilter_4 (Filter%20Attribute%20Elements :AttributeTypes (struct :Type Integer32 :Int32Value 1) :bGenerateOutputDataEvenIfEmpty false :bUseConstantThreshold true :Operator NotEqual :TargetAttribute Assembly))
  ;;[0 0 0] Meshes
  (bind AttributeFilter_2 (Filter%20Attribute%20Elements :AttributeTypes (struct :Type SoftObjectPath) :bGenerateOutputDataEvenIfEmpty false :bUseConstantThreshold true :Operator NotEqual :TargetAttribute Mesh))
  (bind StaticMeshSpawner_0 (Static%20Mesh%20Spawner :bApplyMeshBoundsToPoints false :MeshSelectorParameters (class:/Script/PCG.PCGMeshSelectorByAttribute :AttributeName Mesh :bUseAttributeMaterialOverrides true :MaterialOverrideAttributes [Material] :TemplateDescriptor (struct :bIncludeInHLOD false :bGenerateOverlapEvents true)) :MeshSelectorType /Script/PCG.PCGMeshSelectorByAttribute :StaticMeshComponentPropertyOverrides [(struct :InputSource AllowCollision :PropertyTarget "bUseDefaultCollision") (struct :InputSource CastShadow :PropertyTarget "bCastShadow") (struct :InputSource CastHiddenShadow :PropertyTarget "bCastHiddenShadow") (struct :InputSource DetailMode :PropertyTarget "DetailMode") (struct :InputSource StartCullDistance :PropertyTarget "InstanceStartCullDistance") (struct :InputSource EndCullDistance :PropertyTarget "InstanceEndCullDistance") (struct :InputSource WorldPositionOffsetDisableDistance :PropertyTarget "WorldPositionOffsetDisableDistance") (struct :InputSource IncludeInHLOD :PropertyTarget "bIncludeInHLOD") (struct :InputSource Visible :PropertyTarget "bVisible") (struct :InputSource AffectDistanceFieldLighting :PropertyTarget "bAffectDistanceFieldLighting") (struct :InputSource AffectDynamicIndirectLighting :PropertyTarget "bAffectDynamicIndirectLighting") (struct :InputSource VisibleInRayTracing :PropertyTarget "bVisibleInRayTracing") (struct :InputSource CanEverAffectNavigation :PropertyTarget "bCanEverAffectNavigation") (struct :InputSource RasterizeAsFilledConvexVolume :PropertyTarget "bRasterizeAsFilledConvexVolume") (struct :InputSource CastDynamicShadow :PropertyTarget "bCastDynamicShadow") (struct :InputSource EvaluateWorldPositionOffset :PropertyTarget "bEvaluateWorldPositionOffset") (struct :InputSource ComponentTags :PropertyTarget "AdditionalCommaSeparatedTags")]))
  ;;[0 0 0] Actors
  (bind AttributeFilter_1 (Filter%20Attribute%20Elements :AttributeTypes (struct :Type SoftClassPath) :bGenerateOutputDataEvenIfEmpty false :bUseConstantThreshold true :Operator NotEqual :TargetAttribute Actor))
  (bind SpawnActor_4 (Spawn%20Actor :bSpawnByAttribute true :SpawnAttribute Actor))
  ;;[0 0 0] Skeletal Meshes
  (bind AttributeFilter_3 (Filter%20Attribute%20Elements :AttributeTypes (struct :Type SoftObjectPath) :bUseConstantThreshold true :Operator NotEqual :TargetAttribute SkeletalMesh :ThresholdAttribute @Last))
  (bind AddAttribute_0 (Add%20Attribute :AttributeTypes (struct :Type SoftObjectPath) :OutputTarget AnimBank))
  (bind AddAttribute_1 (Add%20Attribute :AttributeTypes (struct :Type Integer32 :Int32Value -1) :OutputTarget SequenceIndex))
  (bind SkinnedMeshSpawner_0 (Instanced%20Skinned%20Mesh%20Spawner :MeshSelectorParameters (class:/Script/PCG.PCGSkinnedMeshSelector :bUseAttributeMaterialOverrides true :MaterialOverrideAttributes [Material] :MeshAttribute SkeletalMesh :TemplateDescriptor (struct :bRenderStatic true)) :SkinnedMeshComponentPropertyOverrides [(struct :InputSource CastShadow :PropertyTarget "bCastShadow") (struct :InputSource StartCullDistance :PropertyTarget "InstanceStartCullDistance") (struct :InputSource EndCullDistance :PropertyTarget "InstanceEndCullDistance") (struct :InputSource IncludeInHLOD :PropertyTarget "bIncludeInHLOD") (struct :InputSource AffectDistanceFieldLighting :PropertyTarget "bAffectDistanceFieldLighting") (struct :InputSource CastDynamicShadow :PropertyTarget "bCastDynamicShadow") (struct :InputSource AffectDynamicIndirectLighting :PropertyTarget "bAffectDynamicIndirectLighting") (struct :InputSource VisibleInRayTracing :PropertyTarget "bVisibleInRayTracing") (struct :InputSource CastHiddenShadow :PropertyTarget "bCastHiddenShadow") (struct :InputSource Visible :PropertyTarget "bVisible") (struct :InputSource EvaluateWorldPositionOffset :PropertyTarget "bEvaluateWorldPositionOffset") (struct :InputSource DetailMode :PropertyTarget "DetailMode") (struct :InputSource CanEverAffectNavigation :PropertyTarget "bCanEverAffectNavigation")]))
  ;[0 0 0 1 32] Local Root Asset Table
  (bind Subgraph_4 (Subgraph :Graph /PCGBiomeCore/Core/BiomeCore_BuildRootAssetTable.BiomeCore_BuildRootAssetTable))
  ;[0 0 0 1 32] Local Biome Cache
  (bind Subgraph_2 (Subgraph :Graph /PCGBiomeCore/Core/BiomeLocalCache.BiomeLocalCache :SubgraphOverride-param LocalBiomeCache))
  (bind GetActorProperty_0 (Get%20Actor%20Property :PropertyName Enabled))
  (bind Branch_4 (Branch))
  (bind Branch_1 (Branch :bOutputToB-param Debug-DisplayLocalBiome))
  (bind Debug_1 (Debug))
  (bind ExtentsModifier_1 (Extents%20Modifier :Mode Add))
  (bind Debug_2 (Debug))
  ;[1 0.3333 0 0.1] Errors and Warnings
  (bind GetActorData_0 (Get%20Actor%20Data :ActorSelector (struct :ActorFilter AllWorldActors :bMustOverlapSelf true :ActorSelectionTag PCG_BiomeCore) :Mode GetActorReference))
  (bind DataNum_0 (Data%20Count))
  (bind AttributeCast_0 (Attribute%20Cast :OutputType Boolean))
  (bind AttributeBooleanOp_0 (Attribute%20Boolean%20Op :Operation Or :InB LocalPreview))
  (bind Branch_3 (Branch))
  (bind Print%20String_2 (Print%20String :bPrefixWithNode false :bPrefixWithOwner true :bPrintToScreen true :PrintString "No Overlapping BiomeCore, add one or enable Local Preview" :PrintToScreenColor (struct :R 255 :A 255) :PrintToScreenDuration 0.0 :Verbosity Error))
  (bind Branch_6 (Branch :bOutputToB-param LocalPreview :In LocalPreview))
  (bind Print%20String_3 (Print%20String :bPrefixWithNode false :bPrefixWithOwner true :bPrintToScreen true :PrintString "Local Preview is currently enabled" :PrintToScreenColor (struct :G 173 :R 255 :A 255) :PrintToScreenDuration 0.0 :Verbosity Warning))
  ;[0 0 0 1 32] Output to BiomeCore
  (bind Subgraph_6 (Subgraph :Graph /PCGBiomeCore/Core/BiomeCore_PrepareGroundScatterAssetsAndDefinitions.BiomeCore_PrepareGroundScatterAssetsAndDefinitions))
  (bind AttributeFilter_0 (Filter%20Attribute%20Elements :AttributeTypes (struct :Type Integer32) :bUseConstantThreshold true :bWarnOnDataMissingAttribute false :Operator NotEqual :TargetAttribute RuntimeAssemblyHashValue :ThresholdAttribute @Last))
  (bind RemoveEmptyData_0 (Remove%20Empty%20Data))
  (bind GetActorProperty_5 (Get%20Actor%20Property :ActorSelector (struct :ActorSelectionTag PCG_BiomeActor :bSelectMultiple true) :bAlwaysRequeryActors false :bProcessAllComponents false :bSanitizeOutputAttributeName false :bTrackActorsOnlyWithinBounds true :OutputAttributeName BiomeBlendingRange :PropertyName BlendingRange))
  (bind AttributeMathsOp_1 (Attribute%20Maths%20Op :Operation Multiply))
  (bind AttributeMathsOp_0 (Attribute%20Maths%20Op :InputSource1 $Extents))
  (reroute Generators%20Ordered%20Data Subgraph_8.PointsOut)
  (reroute Local%20Biome%20Cache Branch_4.Output%20B)
  (reroute LocalRootAssetTable Subgraph_4.AssetTable)
  (reroute Runtime%20Asset%20Tables%20-%20from%20Assembly Loop_1.RuntimeAssetTables)
  (reroute RuntimeAssetTables Loop_5.RuntimeAssetTables)
  (connect AttributeFilter_3 InsideFilter AddAttribute_0 In)
  (connect AddAttribute_0 Out AddAttribute_1 In)
  (connect AttributeCast_0 Out AttributeBooleanOp_0 InA)
  (connect DataNum_0 Out AttributeCast_0 In)
  (connect RuntimeAssetTables Out AttributeFilter_0 In)
  (connect Runtime%20Asset%20Tables%20-%20from%20Assembly Out AttributeFilter_0 In)
  (connect AttributeFilter_4 InsideFilter AttributeFilter_1 In)
  (connect AttributeFilter_4 InsideFilter AttributeFilter_2 In)
  (connect AttributeFilter_4 InsideFilter AttributeFilter_3 In)
  (connect Branch_2 Output%20B AttributeFilter_4 In)
  (connect Local%20Biome%20Cache Out AttributeMathsOp_0 InA)
  (connect AttributeMathsOp_1 Out AttributeMathsOp_0 InB)
  (connect GetActorProperty_5 Out AttributeMathsOp_1 InA)
  (connect Branch_5 Output%20B AttributeMathsOp_2 InA)
  (connect MutateSeed_1 Out AttributePartition_3 In)
  (connect SplineSampler_1 Out BoundsModifier_2 In)
  (connect Subgraph_8 PointsOut Branch_0 In)
  (connect Branch_4 Output%20B Branch_1 In)
  (connect AttributeMathsOp_2 Out Branch_2 In)
  (connect Branch_5 Output%20A Branch_2 In)
  (connect AttributeBooleanOp_0 Out Branch_3 bOutputToB)
  (connect GetActorData_0 Out Branch_3 In)
  (connect GetActorProperty_0 Out Branch_4 bOutputToB)
  (connect Subgraph_2 LocalBiomeCache Branch_4 In)
  (connect Subgraph_1 Out Branch_5 In)
  (connect GetActorData_0 Out DataNum_0 In)
  (connect Branch_1 Output%20B Debug_1 In)
  (connect ExtentsModifier_1 Out Debug_2 In)
  (connect BoundsModifier_2 Out Difference_3 Differences)
  (connect FilterByType_8 InsideFilter Difference_3 Differences)
  (connect FilterDataByType_7 InsideFilter Difference_3 Differences)
  (connect Subgraph_21 PointsOut Difference_3 Source)
  (connect Branch_1 Output%20B ExtentsModifier_1 In)
  (connect GetActorData_6 Out FilterByTag_5 In)
  (connect FilterByTag_5 InsideFilter FilterByType_7 In)
  (connect FilterByTag_5 InsideFilter FilterByType_8 In)
  (connect GetActorProperty_7 Out FilterDataByTag_1 In)
  (connect GetActorProperty_2 Out FilterDataByTag_2 In)
  (connect Loop_5 Out FilterDataByTag_3 In)
  (connect Branch_0 Output%20A FilterDataByTag_5 In)
  (connect FilterByTag_5 OutsideFilter FilterDataByTag_9 In)
  (connect RemoveEmptyData_2 Out FilterDataByType_2 In)
  (connect FilterDataByTag_9 InsideFilter FilterDataByType_7 In)
  (connect LocalRootAssetTable Out Loop_1 AssetTable)
  (connect FilterDataByTag_3 OutsideFilter Loop_1 In)
  (connect LocalRootAssetTable Out Loop_5 AssetTable)
  (connect AttributePartition_3 Out Loop_5 InPoints)
  (connect GetActorProperty_4 Out Loop_6 BiomeBlendingDensity)
  (connect GetActorProperty_3 Out Loop_6 BiomeBlendingNoise)
  (connect GetActorProperty_1 Out Loop_6 BiomeBlendingRange)
  (connect Subgraph_133 Generators Loop_6 Generators)
  (connect Local%20Biome%20Cache Out Loop_6 LocalBiomeCache)
  (connect FilterDataByTag_5 InsideFilter MergePoints_0 In)
  (connect FilterDataByType_2 InsideFilter MergePoints_0 In)
  (connect Print%20String_0 Out MergePoints_0 In)
  (connect MergePoints_0 Out MutateSeed_1 In)
  (connect Branch_0 Output%20B Print%20String_0 In)
  (connect Branch_3 Output%20A Print%20String_2 In)
  (connect Branch_6 Output%20B Print%20String_3 In)
  (connect AttributeFilter_0 InsideFilter RemoveEmptyData_0 In)
  (connect Subgraph_3 Out RemoveEmptyData_2 In)
  (connect ToPoint_0 Out RemoveEmptyData_3 In)
  (connect AddAttribute_1 Out SkinnedMeshSpawner_0 In)
  (connect AttributeFilter_1 InsideFilter SpawnActor_4 In)
  (connect FilterByType_7 InsideFilter SplineSampler_1 Spline)
  (connect AttributeFilter_2 InsideFilter StaticMeshSpawner_0 In)
  (connect Generators%20Ordered%20Data Out Subgraph_1 GeneratorsOrderedData)
  (connect FilterDataByTag_3 InsideFilter Subgraph_1 In)
  (connect Loop_1 Out Subgraph_1 In)
  (connect CreateConstant_1 Out Subgraph_1 UseAssembliesSourcePointsInDifference)
  (connect FilterDataByTag_5 OutsideFilter Subgraph_3 In)
  (connect RemoveEmptyData_3 Out Subgraph_8 PointsIn)
  (connect LocalRootAssetTable Out Subgraph_8 RootAssetTable)
  (connect Loop_6 SampledPoints Subgraph_21 PointsIn)
  (connect FilterDataByTag_1 InsideFilter Subgraph_133 Assets)
  (connect FilterDataByTag_2 InsideFilter Subgraph_133 LocalAssets)
  (connect Difference_3 Out ToPoint_0 In)
  (output GeneratedPoints Branch_2.Output%20A)
  (output RuntimeGroundScatter:AssetsAndDefinitions Subgraph_6.AssetsAndDefinitions)
  (output RuntimeGroundScatter:Biome AttributeMathsOp_0)
  (output RuntimeMeshScatter:AssetTables RemoveEmptyData_0))
